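; Creative-mode Down But Not Out ability that applies DBNO device settings to downed players.
; On activation, before calling the parent, it reads the GameState DBNO mutator to choose a start effect by team-alert setting, set tenacity, and tune revive-progress decay on the pawn's MultiInteract component.
; ApplyBleeding builds a SetByCaller bleed effect, with magnitude from the mutator's depletion rate or the AthenaGameData curve as fallback.
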(event Ability|EventActivateAbilityFromEvent
  (params (EventData :type Gameplay%20Event%20Data%20Structure%20%28by%20ref%29))
  (CallFunction|SelectStartGameplayEffect)
  (|CallParentFunction EventData))

(event Ability|EventOnEndAbility
  (params (bWasCancelled :type Boolean))
  (|CallParentFunction bWasCancelled) ; This node is disabled and will not be called.\nDrag off pins to build functionality.
)

(fn ApplyBleeding
  (bind _returnvalue (Ability|GetAvatarActorfromActorInfo))
  (bind _returnvalue_2 (Fortnite|Athena|GameState|GetGameStateAthenafromContext _returnvalue))
  (bind _aspawn (Utilities|Casting|CastToPawn _returnvalue)
    (:then
      (bind _asfort_player_state (Utilities|Casting|CastToFortPlayerState (Variables|/Script/Engine.Pawn|GetPlayerState _aspawn))
        (:then
          (if (and (>= (.TenacityDepletionRate (Utilities|Struct|BreakDBNOCustomSettings (Class|FortGameStateAthena|EvaluateDBNOMutator _returnvalue_2 _asfort_player_state))) 0.0) (.ReturnValue (Class|FortGameStateAthena|EvaluateDBNOMutator _returnvalue_2 _asfort_player_state)))
            (LocalVariables|SetBleedingMagnitude (.TenacityDepletionRate (Utilities|Struct|BreakDBNOCustomSettings (Class|FortGameStateAthena|EvaluateDBNOMutator _returnvalue_2 _asfort_player_state))))
            (else
              (bind _outxy (DataTable|EvaluateCurveTableRow "/Game/Athena/Balance/DataTables/AthenaGameData.AthenaGameData" "Default.DBNO.BleedDamage")
                (:RowFound
                  (LocalVariables|SetBleedingMagnitude _outxy))
                (:RowNotFound
                  _))
              (bind _returnvalue_3 (Ability|GameplayEffect|AssignTagSetbyCallerMagnitude (Ability|MakeOutgoingGameplayEffectSpec "/CRD_DBNO/Abilities/GE_Creative_DBNO_Bleed.GE_Creative_DBNO_Bleed_C" 0.0) "(TagName=\"SetByCaller.DBNO.Bleed\")" (LocalVariables|GetBleedingMagnitude)))
              (bind _returnvalue_4 (Ability|ApplyGameplayEffectSpecToOwner _returnvalue_3))
              (return _returnvalue_4))))
        (:CastFailed
          _)))
    (:CastFailed
      _)))

(fn SelectStartGameplayEffect
  (bind _returnvalue (Ability|GetAvatarActorfromActorInfo))
  (bind _returnvalue_2 (Fortnite|Athena|GameState|GetGameStateAthenafromContext _returnvalue))
  (bind _decayrow (LocalVariables|GetDecayRow))
  (Utilities|FlowControl|Sequence
    (:then_0
      (bind _aspawn (Utilities|Casting|CastToPawn _returnvalue))
      (bind _asfort_player_state (Utilities|Casting|CastToFortPlayerState (Variables|/Script/Engine.Pawn|GetPlayerState _aspawn)))
      (if (.ReturnValue (Class|FortGameStateAthena|EvaluateDBNOMutator _returnvalue_2 _asfort_player_state))
        (LocalVariables|SetDBNOStartClassToApply (select (.bAlertTeam (Utilities|Struct|BreakDBNOCustomSettings (Class|FortGameStateAthena|EvaluateDBNOMutator _returnvalue_2 _asfort_player_state))) "/Game/Athena/MappedEffects/GE_Athena_DBNO_Start.GE_Athena_DBNO_Start_C" "/CRD_DBNO/Abilities/GE_Creative_DBNO_Start_NoCue.GE_Creative_DBNO_Start_NoCue_C"))
        (Variables|Default|SetTenacityType (.TenacityType (Utilities|Struct|BreakDBNOCustomSettings (Class|FortGameStateAthena|EvaluateDBNOMutator _returnvalue_2 _asfort_player_state))))
        (Variables|Default|SetTenacityAmount (.TenacityAmount (Utilities|Struct|BreakDBNOCustomSettings (Class|FortGameStateAthena|EvaluateDBNOMutator _returnvalue_2 _asfort_player_state))))
        ;[0.4242 1 0.2517 0.5] Apply values set up by DBNO device
        (bind _asplayer_pawn_athena (Utilities|Casting|CastToPlayerPawn_Athena (Ability|GetAvatarActorfromActorInfo)))
        (Variables|/Script/FortniteGame.FortActorComponent_MultiInteract|SetbAllowGradualReset (.bAllowDecayAndReset (Utilities|Struct|BreakDBNOCustomSettings (Class|FortGameStateAthena|EvaluateDBNOMutator _returnvalue_2 _asfort_player_state))) (Variables|/Game/Athena/PlayerPawn_Athena_Generic_Parent.PlayerPawn_Athena_Generic_Parent_C|GetMultiInteractActorComponent _asplayer_pawn_athena))
        (ScalableFloat|SetValue _decayrow (.ReviveProgressDecayMulti (Utilities|Struct|BreakDBNOCustomSettings (Class|FortGameStateAthena|EvaluateDBNOMutator _returnvalue_2 _asfort_player_state))))
        (Variables|/Script/FortniteGame.FortActorComponent_MultiInteract|SetGradualCounterMultiplier _decayrow (Variables|/Game/Athena/PlayerPawn_Athena_Generic_Parent.PlayerPawn_Athena_Generic_Parent_C|GetMultiInteractActorComponent _asplayer_pawn_athena)) ; The Multi Interact component manually checks this scalable float
        (Variables|/Script/FortniteGame.FortActorComponent_MultiInteract|SetbEndClientLerpWhenDone (.bEndClientLerpWhenReviveInterrupted (Utilities|Struct|BreakDBNOCustomSettings (Class|FortGameStateAthena|EvaluateDBNOMutator _returnvalue_2 _asfort_player_state))) (Variables|/Game/Athena/PlayerPawn_Athena_Generic_Parent.PlayerPawn_Athena_Generic_Parent_C|GetMultiInteractActorComponent _asplayer_pawn_athena))
        (Class|FortActorComponentMultiInteract|ResetCachedInterpSpeed (Variables|/Game/Athena/PlayerPawn_Athena_Generic_Parent.PlayerPawn_Athena_Generic_Parent_C|GetMultiInteractActorComponent _asplayer_pawn_athena))))
    (:then_1
      (Variables|Default|SetDBNOStart_GE_Class (LocalVariables|GetDBNOStartClassToApply)))))
