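; Creative variant of the DBNO interrogate-player reveal ability, letting custom DBNO settings govern whether interrogation reveals happen.
; Overrides Reveal to get the avatar pawn's Fort player state and evaluate the game state's DBNO mutator for it.
; The parent reveal is gated on bAllowInterrogationReveal being set or the mutator evaluation returning false.
; Activation from event simply forwards to the parent ability.
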(event Ability|EventActivateAbilityFromEvent
  (params (EventData :type Gameplay%20Event%20Data%20Structure%20%28by%20ref%29))
  (|CallParentFunction EventData))

(event Ability|EventOnEndAbility
  (params (bWasCancelled :type Boolean))
  (|CallParentFunction bWasCancelled) ; This node is disabled and will not be called.\nDrag off pins to build functionality.
)

(fn Reveal
  (bind _returnvalue (Ability|GetAvatarActorfromActorInfo))
  (bind _aspawn (Utilities|Casting|CastToPawn _returnvalue)
    (:then
      (bind _asfort_player_state (Utilities|Casting|CastToFortPlayerState (Variables|/Script/Engine.Pawn|GetPlayerState _aspawn))
        (:then
          (if (or (.bAllowInterrogationReveal (Utilities|Struct|BreakDBNOCustomSettings (Class|FortGameStateAthena|EvaluateDBNOMutator (Fortnite|Athena|GameState|GetGameStateAthenafromContext _returnvalue) _asfort_player_state))) (not (.ReturnValue (Class|FortGameStateAthena|EvaluateDBNOMutator (Fortnite|Athena|GameState|GetGameStateAthenafromContext _returnvalue) _asfort_player_state))))
            _))
        (:CastFailed
          _))
      (|CallParentFunction))
    (:CastFailed
      _)))
